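# Niagara anim trail effect for the male Armadillo pickaxe, pairing a soft ribbon trail with hot pink sparks.
# Trail_soft uses the FN AnimTrail modules with skeletal mesh bones or sockets, feeding two ribbon renderers plus separately faded bolt width and color values.
# Trail tint and opacity scale with time of day.
# Sparks spawn after a brief delay, inherit velocity, then fall and shrink under gravity, drag and curl noise.
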NiagaraSystem /BRCosmetics/Weapons/FORT_Melee/Pickaxe_Armadillo_Male/FX/NS_ArmadilloMale_Pickaxe_Animtrail
dependencies {
 usedDataInterfaces [&NiagaraDataInterfaceCurve, &NiagaraDataInterfaceCurlNoise]
 usedDynamicInputs []
 usedModules [&SystemState]
}
emitters {
 Sparks {
  emitterUpdateScript [
   &EmitterState(Life Cycle Mode=::NewEnumerator1, Loop Behavior=::NewEnumerator1, Loop Delay=0.05, Loop Duration=2, UseLoopDelay=true)
   &SpawnRate(SpawnRate=&RandomRangeFloat(Minimum=50, Maximum=60), Use Spawn Probability=false)
  ]
  particleSpawnScript [
   &InitializeParticle(
    Color Maximum=<rgba 10 0.2 0.3633 1>
    Color Minimum=<rgba 10 0.2 0.69 1>
    Color Mode=::NewEnumerator2
    Lifetime=&RandomRangeFloat(Minimum=0.4, Maximum=0.5)
    Position Offset=&RandomRangeVector(Minimum=<10 10 70>, Maximum=<-10 -10 100>)
    Position Offset Coordinate Space=::Local
    Sprite Size=&RandomRangeVector2D(Minimum=<4 12>, Maximum=<7 17>)
    Sprite Size Mode=::NewEnumerator4
    UsePositionOffset=true
    Write Material Random=true
   )
   &InheritVelocity(Inherited Velocity Amount Scale=<2 2 2>, Speed Limit=true, UseSpeedThreshold=false) {module=InheritVelocity001}
  ]
  particleUpdateScript [
   &ParticleState()
   &GravityForce(Gravity=<0 0 -128>)
   &Drag(Drag=5, Use Linear Drag=true, Use Rotational Drag=false)
   &CurlNoiseForce(Noise Frequency=100, Pan Noise=false)
   &SolveForcesAndVelocity(Clamp Velocity=false, Limit Acceleration=false)
   &ScaleColor(
    Scale Alpha=&FloatFromCurve(FloatCurve=DataInterface() {propertyValues {Curve {keys [<at 0 1 RCIM_Cubic>, {arriveTangent=-3.2218, interpMode=RCIM_Cubic, leaveTangent=-3.2218, tangentMode=RCTM_User, time=1, value=0}]}}})
    ScaleA=true
    ScaleRGB=true
   )
   &ScaleSpriteSize(Scale Factor=&Vector2DFromFloat(Value=&FloatFromCurve(FloatCurve=DataInterface() {propertyValues {Curve {keys [<at 0 1 RCIM_Cubic>, {arriveTangent=-3.6304, interpMode=RCIM_Cubic, leaveTangent=-3.6304, tangentMode=RCTM_User, time=1, value=0}]}}})))
  ]
  rendererClasses [&NiagaraSpriteRendererProperties]
  renderers [{rendererClass=&NiagaraSpriteRendererProperties}]
 }
 Trail_soft {
  emitterUpdateScript [
   &EmitterState()
   &FN_SpawnAnimTrail_v2(AnimTrailWidthMode=::FromSecond)
  ]
  particleSpawnScript [
   /Game/Effects/Niagara/Modules/AnimTrail/FN_InitializeAnimTrail/FN_InitializeAnimTrailBoneIndecies(Second Bone or Socket Index=1)
   /Game/Effects/Niagara/Modules/AnimTrail/FN_InitializeAnimTrail/FN_InitializeAnimTrail(
    AnimTrailWidth=1.5
    Color=&LerpLinearColors(
     StartColor=<rgba 1 0.8781 0.8786 0.2>
     EndColor=<rgba 1 0.2948 0.4241 0.2>
     LerpFactor=&FNTimeOfDayScalar(Night Scale=0, Morning Scale=0.5, Evening Scale=0.5)
    )
    Lifetime=0.2
    Position Offset:Vector3f=&TransformVector(Vector=<0 0 0>, Source Space=::Simulation, Destination Space=::Local)
   )
   /SetVariables(Particles.BoltWidth:NiagaraFloat=&Multiply_Float(A=$Particles.RibbonWidth, B=0.6)) {isSetParameters=true}
   /SetVariables(Particles.RibbonU1Override:NiagaraFloat=&Multiply_Float(A=$Emitter.Age, B=10)) {isSetParameters=true}
   /SetVariables() {isSetParameters=true}
  ]
  particleUpdateScript [
   &ScaleColor(
    Scale Alpha=&Multiply_Float(
     A=&FloatFromCurve(FloatCurve=DataInterface() {propertyValues {Curve {keys [
          {interpMode=RCIM_Cubic, leaveTangent=-2.3646, tangentMode=RCTM_Break, time=0, value=1}
          <at 0.6689 0.0233>
          {arriveTangent=-0.2139, interpMode=RCIM_Cubic, tangentMode=RCTM_Break, time=1, value=0}
         ]}}})
     B=&FNTimeOfDayScalar(Night Scale=0.1, Morning Scale=0.5, Evening Scale=0.5)
    )
    ScaleA=true
    ScaleRGB=true
   ) {module=ScaleColor002}
   &FN_UpdateAnimTrail_v2()
   /SetVariables(Particles.BoltWidth:NiagaraFloat=&Multiply_Float(A=$Particles.RibbonWidth, B=0.6)) {isSetParameters=true}
   /SetVariables(Particles.BoltColor:LinearColor=&MultiplyLinearColor(
     Linear Color=$Particles.Initial.BoltColor
     Scale Factor=&MakeVector4(
      X=1
      Y=1
      Z=1
      W=&FloatFromCurve(FloatCurve=DataInterface() {propertyValues {Curve {keys [<at 0 1>,<at 0.35 0>]}}})
     )
    )) {isSetParameters=true}
  ]
  rendererClasses [&NiagaraRibbonRendererProperties]
  renderers [
   {rendererClass=&NiagaraRibbonRendererProperties}
   {rendererClass=&NiagaraRibbonRendererProperties, rendererIndex=1}
  ]
 }
}
summary {
 systemName=NS_ArmadilloMale_Pickaxe_Animtrail
 emitters [
  {emitterName=Trail_soft, rendererClasses [&NiagaraRibbonRendererProperties]}
  {emitterName=Sparks, rendererClasses [&NiagaraSpriteRendererProperties]}
 ]
 userVariables [
  User.RibbonIDOffset {type=NiagaraInt32}
  User.SkeletalMesh {
   type=NiagaraDataInterfaceSkeletalMesh
   defaultValue {
    struct=NiagaraExt_VariableValue_DataInterface
    value {dataInterface=&self:NiagaraDataInterfaceSkeletalMesh_2}
   }
  }
 ]
}
